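; Cosmetic stat watcher component that makes a teammate-reactive sound trigger across the local player's team.
; After begin play it checks whether the owning pawn is on the local team and, if so, broadcasts an init event registering it with a reactive group manager.
; PlayLibrarySound broadcasts a reactive event carrying the group and sound library tag, only for the local team.
; EndPlay broadcasts the init event again to remove the player.
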(event EventFinishedBeginPlay
  (CallFunction|CheckTeamSize))

(event Custom|CheckTeamSize
  ;[0.15 0.0165 0.0165 0.5] Only add for local team - check team size, no need for manager if team size is 1
  (bind _output_get (Variables|Default|SetLocalPlayerTeam (Team|OnLocalPlayersTeam (CustomCharacter|GetAssociatedPlayerPawn (Variables|Private|GetPartsModifier)))))
  (if _output_get
    (CallFunction|AddPlayertoManager)))

(event Custom|AddPlayerToManager
  ; Add player to the manager, create manager if not currently existing
  (GameplayEvent|BroadcastEvent (Variables|Private|GetInitEvent) 0 (Utilities|Struct|MakeGameplayEventTeammateReactiveInit (CustomCharacter|GetAssociatedPlayerPawn (Variables|Private|GetPartsModifier)) (Variables|StatWatcher|GetReactiveGroup) true)))

(event EventPlayLibrarySound
  (params (StatName :type Name :default "None") (StatValue :type Integer) (LibraryTag :type Sound%20Library%20Tag%20Structure%20%28by%20ref%29))
  ; Trigger reactivity for whole team, only for local team
  (Variables|Default|GetLocalPlayerTeam
    (:then
      (GameplayEvent|BroadcastEvent (Variables|Private|GetReactiveEvent) 0 (Utilities|Struct|MakeGameplayEventTeammateReactive (Variables|StatWatcher|GetReactiveGroup) (Variables|StatWatcher|GetSoundLibraryTag))))
    (:else)))

(event EventEndPlay
  (params (EndPlayReason :type EEndPlayReason%20Enum :default "Destroyed"))
  ; Remove player from reactive group
  (|CallParentFunction EndPlayReason)
  (GameplayEvent|BroadcastEvent (Variables|Private|GetInitEvent) 0 (Utilities|Struct|MakeGameplayEventTeammateReactiveInit (CustomCharacter|GetAssociatedPlayerPawn (Variables|Private|GetPartsModifier)) (Variables|StatWatcher|GetReactiveGroup))))
